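# Lit material for the Creative Map Indicator device's compass display, showing a rotating arrow and fixed cardinal markings.
# A vertex-interpolated atan2 of object forward against world axes rotates the arrow UVs, which sine-timed bursts of noise distort.
# Arrow and cardinals from one texture's channels composite over panning noise, gradient and grid.
# Distance-faded scanlines modulate base color and emissive, and the attributes pass through MF_CreativeMoveTool.
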Material /Game/Creative/Devices/MapIndicator/Materials/M_MapIndicator_Compass
graph {
 properties {bUseMaterialAttributes=true, pixelDepthOffsetMode=PDOM_Legacy, usedShadingModels=Default Lit}
 nodes [
  n1 = TextureCoordinate()
  n2 = Frac(None=Multiply(A=ComponentMask(None=@n1) {g=true}, B=ScalarParameter() {defaultValue=20, parameterName=ScanlineRepeat}))
  n3 = Constant() {r=0.5}
  n4 = ScalarParameter() {defaultValue=0.6, parameterName=ScanlineRadius}
  n5 = ScalarParameter() {defaultValue=4, parameterName=ScanlineHarndess}
  n6 = LinearInterpolate(
   A=Multiply(A=SphereMask(A=@n2, B=@n3, Radius=@n4, Hardness=@n5), B=ScalarParameter() {defaultValue=0.125, parameterName=scanline opacity})
   Alpha=MaterialFunctionCall(
    Position=ActorPositionWS()
    Fade_Start_Distance=ScalarParameter() {defaultValue=700, parameterName=Fade_Start_Distance}
    Fade_End_Distance=ScalarParameter() {defaultValue=200, parameterName=Fade_End_Distance}
   ) {materialFunction=&MF_Distance_SyncedFade_FX}.Result
  )
  n7 = Multiply(
   A=SphereMask(A=TextureCoordinate(), B=Constant() {r=0.5}, Radius=Constant() {r=0.6}, Hardness=Constant())
   B=ScalarParameter() {defaultValue=2, parameterName=GradientScale}
  )
  n8 = VectorParameter() {parameterName=NoiseAScalePan, defaultValue=<rgba 0 0.66 0 0.1267>}
  n9 = Multiply(A=TextureCoordinate(), B=AppendVector(A=@n8.R, B=@n8.R))
  n10 = AppendVector(A=@n8.B, B=@n8.B)
  n11 = VectorParameter() {parameterName=NoiseBScalePan, defaultValue=<rgba 0 0.5 0 -0.0814>}
  n12 = Multiply(A=TextureCoordinate(), B=AppendVector(A=@n11.R, B=@n11.R))
  n13 = AppendVector(A=@n11.B, B=@n11.B)
  n14 = Multiply(
   A=TextureSample(UVs=Panner(Coordinate=@n9, Speed=@n10) {bFractionalPart=true}) {samplerType=SAMPLERTYPE_Normal, texture=/Game/Global/Textures/Default/Noise/64x64Noise}.R
   B=TextureSample(UVs=Panner(Coordinate=@n12, Speed=@n13) {bFractionalPart=true}) {samplerType=SAMPLERTYPE_Normal, texture=&64x64Noise}.G
  )
  n15 = VectorParameter() {parameterName=NoiseCScalePan, defaultValue=<rgba 4 4 15 0>}
  n16 = Panner(Coordinate=Multiply(A=TextureCoordinate(), B=AppendVector(A=@n15.R, B=@n15.R)), Speed=AppendVector(A=@n15.B, B=@n15.B)) {bFractionalPart=true}
  n17 = Add(
   A=@n14
   B=Multiply(A=@n14, B=TextureSample(UVs=@n16) {samplerType=SAMPLERTYPE_Normal, texture=&64x64Noise}.G)
  )
  n18 = Multiply(A=TextureCoordinate(), B=ScalarParameter() {defaultValue=2, parameterName=GridTile})
  n19 = Add(
   A=Multiply(A=Add(A=@n7, B=@n17), B=VectorParameter() {parameterName=StaticNoiseColor, defaultValue=<rgba 0.1126 0.3854 0.1917 0>}.RGB)
   B=Multiply(
    A=TextureSample(UVs=@n18) {texture=/Game/Weapons/Prototype/ObjectMover/Textures/T_CreativeMoveTool_Grid}.R
    B=VectorParameter() {parameterName=GridColor, defaultValue=<rgba 1.4792 4 0.8542 0>}.RGB
   )
  )
  n20 = Normalize(VectorInput=Transform(None=VectorParameter() {parameterName=ObjectForward, defaultValue=<rgba -1 0 0 0>}.RGB) {transformSourceType=TRANSFORMSOURCE_Local})
  n21 = DotProduct(A=VectorParameter() {parameterName=WorldForward, defaultValue=<rgba -1 0 0 1>}.RGB, B=@n20)
  n22 = DotProduct(A=@n20, B=VectorParameter() {parameterName=WorldRight, defaultValue=<rgba 0 -1 0 1>}.RGB)
  n23 = Rotator(Time=Add(
    A=Arctangent2Fast(Y=@n21, X=@n22) {desc=-pi to pi}
    B=MaterialFunctionCall() {materialFunction=&Pi}.Result
   ) {desc=0 to radian}) {speed=1}
  n24 = TextureCoordinate() {uTiling=0}
  n25 = AppendVector(
   A=TextureSample(UVs=Panner(Coordinate=@n24) {bFractionalPart=true, speedX=0.025, speedY=0.3844}) {samplerType=SAMPLERTYPE_Normal, texture=&64x64Noise}.R
   B=Constant()
  )
  n26 = Time()
  n27 = Sine(None=@n26) {period=0.355}
  n28 = Sine(None=@n26) {period=0.8755}
  n29 = Power(Base=Saturate(None=Add(A=@n27, B=@n28))) {constExponent=5}
  n30 = Multiply(A=ScalarParameter() {defaultValue=0.05, parameterName=DistortScale}, B=ComponentMask(None=VertexInterpolator(VS=@n29).PS) {r=true})
  n31 = TextureSample(UVs=Add(A=VertexInterpolator(VS=@n23).PS, B=Multiply(A=@n25, B=@n30))) {samplerSource=SSM_Clamp_WorldGroupSettings, texture=/Game/Creative/Devices/QuestDevice/Textures/T_questDevice-Compass}
  n32 = TextureSample(UVs=MaterialFunctionCall(Texture Scale=ScalarParameter() {defaultValue=1, parameterName=CardinalsScaleUV}) {materialFunction=&ScaleUVsByCenter}.UVs) {samplerSource=SSM_Clamp_WorldGroupSettings, texture=&T_questDevice-Compass}
  n33 = Add(
   A=Multiply(A=@n31.G, B=VectorParameter() {parameterName=ArrowColor, defaultValue=<rgba 3 0 0 1>}.RGB)
   B=Multiply(A=@n32.R, B=VectorParameter() {parameterName=CardinalsColor, defaultValue=<rgba 1.4688 6 2.4276 0>}.RGB)
  )
  n34 = Add(A=@n31.G, B=@n32.R)
  n35 = SetMaterialAttributes!(
   MaterialAttributes
   Base Color=Multiply(A=VectorParameter() {parameterName=BaseColor, defaultValue=<rgba 0.047 1 0.5931 0>}.RGB, B=@n6)
   Metallic=ScalarParameter() {parameterName=Metallic}
   Specular=ScalarParameter() {defaultValue=0.8, parameterName=Specular}
   Roughness=ScalarParameter() {defaultValue=0.1, parameterName=Roughness}
   Emissive Color=Multiply(A=@n6, B=LinearInterpolate(A=@n19, B=@n33, Alpha=@n34))
  ) {attributeSetTypes [
    #1
    #2
    #3
    #4
    #5
   ]}
  MaterialFunctionCall(MA In=@n35) {materialFunction=&MF_CreativeMoveTool}
 ]
}
